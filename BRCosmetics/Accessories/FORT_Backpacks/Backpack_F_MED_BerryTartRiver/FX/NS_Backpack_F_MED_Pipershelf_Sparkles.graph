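# Niagara sparkle effect for the Pipershelf backpack cosmetic, with a single emitter spawning short-lived, lightly tinted sprites.
# SkeletalMeshLocation positions particles on the system's skeletal mesh, offset by picks from a large baked point array.
# Alpha flickers over lifetime and fades in with emitter age, while a shared PiperShelf time-of-day module drives color and alpha, and sprites scale by time of day.
# Three sprite renderers accompany derived center, corona and flicker flare parameters.
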NiagaraSystem /BRCosmetics/Accessories/FORT_Backpacks/Backpack_F_MED_BerryTartRiver/FX/NS_Backpack_F_MED_Pipershelf_Sparkles
dependencies {
 usedDataInterfaces [
  &NiagaraDataInterfaceArrayFloat3
  &NiagaraDataInterfaceCurve
  &NiagaraDataInterfaceSkeletalMesh
  &NiagaraDataInterfaceColorCurve
 ]
 usedDynamicInputs [&Lerp_Float, &MakeVector2D]
 usedModules [
  &self:SystemSpawnScript.SystemScriptSource.SystemScriptGraph.NiagaraNodeAssignment_0.SetVariables_#1
  &SystemState
  /BRCosmetics/Characters/Player/Female/Medium/Bodies/F_MED_PiperShelf_Pearl/FX/TimeOfDayGlobal
 ]
}
emitters {Sparkles {
  emitterUpdateScript [
   &EmitterState()
   &SpawnRate(SpawnRate=7, Use Spawn Probability=true)
  ]
  particleSpawnScript [
   &InitializeParticle(
    AdjustAlpha=false
    AdjustHue=true
    AdjustSaturation=false
    AdjustValue=false
    Color=<rgba 0.843 0.8 1 1>
    Color Mode=::NewEnumerator3
    Lifetime Max=0.4
    Lifetime Min=0.3
    Lifetime Mode=::NewEnumerator1
    Sprite Size Max=&Multiply_Vector2D_ByFloat(Float=0.75, Vector2D=<45 45>)
    Sprite Size Min=&Multiply_Vector2D_ByFloat(Float=0.75, Vector2D=<35 35>)
    Sprite Size Mode=::NewEnumerator2
    UsePositionOffset=false
    Write Material Random=true
   )
   &CameraOffset(Camera Offset Amount=&FN_FrontEndFloatModifier(Gameplay=10, Front End=3))
  ]
  particleUpdateScript [
   &ParticleState()
   &SkeletalMeshLocation(
    Bone Sampling Mode=::NewEnumerator4
    MaskSpawnGroup=false
    Sampled Position Offset=&SelectVectorFromArray(
     Array Sampling Mode=::NewEnumerator0
     Direct Array Index=?
     Vector Selection Array=DataInterface() {propertyValues {FloatData [
        <-0.6292 -17.9134 -11.2452>
        <0.0785 -17.4247 -11.8235>
        <2.8176 -15.4428 1.0043>
        <2.9062 -15.8126 1.3614>
        <0.5114 -15.9953 1.6032>
        <1.5892 -16.9127 3.6846>
        <1.2435 -17.0796 4.1803>
        <0.8635 -17.2148 3.8392>
        <-0.0396 -15.3179 6.1548>
        <-1.9742 -16.196 5.063>
        <-0.7218 -17.1107 2.8894>
        <-1.9419 -16.6755 4.2472>
        <-3.0748 -16.2806 4.2718>
        <-4.3529 -16.0793 3.6803>
        <-6.2132 -15.3022 0.9429>
        <-6.0427 -15.2063 3.7525>
        <-7.2953 -15.1975 2.1823>
        <-9.0495 -13.71 2.174>
        <-8.8351 -14.1003 2.3082>
        <-8.4032 -14.2915 2.5573>
        <-8.041 -14.0222 2.9024>
        <-5.2591 -14.7828 4.5091>
        <-4.0661 -14.0332 5.5977>
        <-5.1902 -14.3066 4.8884>
        <-4.8494 -13.5499 5.6637>
        <-6.4773 -13.4212 4.9005>
        <-7.2841 -13.82 3.77>
        <-8.0094 -13.3407 3.6686>
        <-8.3655 -11.0662 4.4283>
        <-7.7544 -10.9853 4.786>
        <-8.8312 -11.6175 3.4451>
        <-9.8373 -7.572 3.3108>
        <-9.1972 -4.6444 2.7511>
        <-8.6803 -2.7933 0.9657>
        <-8.2687 -1.8669 0.0946>
        <-8.0764 -1.6691 -1.389>
        <-7.4129 -1.327 -3.389>
        <-7.2846 -1.923 1.4587>
        <-5.4672 -0.5294 -2.8976>
        <-5.2168 -0.4482 -1.297>
        <-5.4453 -0.7727 0.6202>
        <-4.9086 -0.3825 -0.6319>
        <-4.644 -0.3533 0.5587>
        <-3.5133 -0.2379 0.8793>
        <-3.2867 0.0062 -0.0581>
        <-5.6825 -1.0893 -4.684>
        <-5.3317 -1.051 -4.8224>
        <-4.3891 -0.6513 -4.3252>
        <-2.2619 -0.246 -3.9027>
        <0.6051 -0.477 -5.4709>
        <0.8494 -0.4658 -5.2519>
        <0.9008 -0.2242 -4.3122>
        <0.2678 -0.2736 -4.5842>
        <1.4481 -0.1234 -3.6913>
        <-1.189 -0.0313 -3.7149>
        <-0.7262 0.1561 -2.5915>
        <-1.604 0.1269 -2.1412>
        <-1.0979 0.1818 -2.0246>
        <-2.1458 0.1768 -1.7356>
        <-1.5742 0.1905 -1.0435>
        <-0.0258 0.1811 -1.1472>
        <0.8149 0.2051 -1.4024>
        <1.7227 0.0553 -2.7646>
        <2.9007 -0.2892 -3.625>
        <2.5 0.0487 -0.828>
        <3.4022 -0.0182 -2.1374>
        <4.7687 -0.3832 -3.1413>
        <6.5595 -0.9484 -3.4098>
        <7.9274 -1.6306 -2.8085>
        <6.3341 -0.7495 -2.184>
        <6.0413 -0.775 -0.1497>
        <1.394 -0.0422 0.5478>
        <2.069 -0.0957 0.8518>
        <0.5816 0.0138 0.5959>
        <-0.2144 0.0224 0.5888>
        <0.297 -0.0059 1.1249>
        <0.4151 -0.5221 2.5308>
        <0.7833 -0.4966 2.4285>
        <-1.1165 -0.8732 3.1706>
        <-0.0011 -0.9605 3.4422>
        <2.6247 -0.8254 2.8108>
        <3.4928 -1.4747 3.3646>
        <2.4131 -1.5224 4.1839>
        <6.0545 -1.436 2.0366>
        <6.2303 -1.5161 1.9745>
        <7.3557 -1.7688 1.1278>
        <8.7155 -2.773 1.1866>
        <9.1599 -3.409 1.8556>
        <10.8268 -6.9746 1.7733>
        <8.6447 -5.4293 3.6489>
        <7.1994 -3.5783 3.7899>
        <4.4386 -2.1088 3.8441>
        <5.0591 -2.6237 4.1445>
        <4.8754 -2.4669 4.0605>
        <3.7418 -3.4095 5.5018>
        <0.9219 -2.4258 5.4808>
        <-0.1823 -2.4983 5.6404>
        <0.6724 -3.5482 6.1993>
        <1.658 -3.7402 6.0999>
        <3.3956 -3.7313 5.8773>
        <3.6126 -4.7558 6.3078>
        <4.2782 -4.9417 6.2832>
        <5.8935 -4.8437 5.7661>
        <8.0966 -6.1537 4.6612>
        <6.7866 -7.5594 5.7978>
        <10.0312 -9.6964 2.9369>
        <10.1816 -10.2026 2.6805>
        <10.4365 -10.7369 2.0034>
        <9.6168 -10.4829 3.3883>
        <7.1135 -10.3197 5.5393>
        <2.1703 -7.3369 7.4651>
        <2.2258 -6.8115 7.3749>
        <-1.1059 -2.8202 5.5975>
        <-1.3449 -4.1354 6.56>
        <-0.9593 -4.9606 7.1803>
        <-0.9394 -5.0709 7.2095>
        <0.1153 -5.62 7.2505>
        <-1.5592 -5.9088 7.2544>
        <-2.2773 -6.4525 7.1592>
        <-3.0646 -6.0088 6.9289>
        <-1.9534 -1.8715 4.5739>
        <-2.8406 -1.1579 3.5147>
        <-2.3079 -0.4965 2.2925>
        <-2.237 -0.0272 0.2804>
        <-3.5445 -0.6413 2.0447>
        <-5.9072 -2.9174 3.9324>
        <-7.1952 -3.3745 3.4869>
        <-6.4952 -4.4325 5.0278>
        <-4.4827 -7.3788 6.7305>
        <-3.1548 -7.0952 7.2429>
        <-2.8338 -7.7947 7.2061>
        <-2.508 -8.309 7.236>
        <-3.1226 -8.7316 7.353>
        <-5.3619 -8.4191 6.5108>
        <-0.096 -8.095 7.5433>
        <-0.2519 -8.625 7.4657>
        <1.3002 -8.5223 7.5313>
        <2.3676 -8.6352 7.2962>
        <3.863 -9.5021 6.975>
        <1.8076 -10.3504 7.2516>
        <6.6978 -12.2856 5.0115>
        <7.834 -13.017 4.092>
        <9.1874 -11.1987 3.5632>
        <10.0466 -12.4914 1.7491>
        <9.5371 -13.7319 1.315>
        <8.245 -14.7613 0.1701>
        <6.998 -14.5613 -0.0919>
        <7.6798 -14.8543 0.7068>
        <5.0857 -15.9 2.0229>
        <6.5809 -15.4557 2.7596>
        <5.514 -15.6216 3.0853>
        <4.7634 -15.727 3.4786>
        <2.5963 -15.983 5.1258>
        <1.9122 -15.7466 5.4534>
        <2.3003 -15.015 5.7202>
        <3.8494 -15.1524 5.1027>
        <4.5342 -14.5423 5.1495>
        <4.5208 -13.7701 5.6252>
        <7.1979 -13.7559 3.9079>
        <8.1238 -14.815 1.1073>
        <7.3766 -15.1446 1.1148>
        <7.4252 -15.0901 2.3071>
        <-8.6382 -14.6107 -4.0471>
        <-8.9476 -14.4571 -2.8338>
        <-9.2304 -14.5567 -2.9471>
        <-8.8995 -14.386 -2.5022>
        <-8.6913 -14.4447 -0.7431>
        <-9.2123 -14.673 -1.5588>
        <-9.9094 -12.7499 1.8361>
        <-10.4886 -11.6864 1.7059>
        <-10.3971 -9.4868 2.6207>
        <-10.1918 -9.0102 2.7894>
        <-10.8352 -9.5401 1.7539>
        <-10.5049 -8.9459 2.2256>
        <-11.017 -8.8438 1.5315>
        <-9.6505 -4.6181 1.9211>
        <-11.0015 -6.7321 1.2224>
        <-11.0261 -6.3332 0.977>
        <-9.7731 -3.9015 0.9699>
        <-9.9923 -3.4852 0.1413>
        <-11.2319 -5.1024 -1.3303>
        <-9.2635 -2.3248 -1.9446>
        <-9.1661 -2.4055 -3.0257>
        <-7.986 -1.8197 -4.2871>
        <-6.7341 -1.6071 -5.1698>
        <-7.8437 -2.1904 -5.1755>
        <-8.2549 -2.9052 -6.23>
        <-7.4253 -2.7516 -6.3669>
        <-9.8798 -4.1357 -6.2841>
        <-8.8908 -4.3251 -7.4577>
        <-6.9356 -2.8717 -6.8261>
        <-2.1802 -1.5459 -6.9871>
        <2.1183 -1.1849 -6.3084>
        <2.529 -1.3917 -6.5808>
        <1.3799 -1.8082 -7.3211>
        <5.4116 -1.8743 -6.5611>
        <6.6254 -1.6111 -5.3829>
        <6.7998 -2.7617 -6.8307>
        <6.1765 -2.8365 -7.1888>
        <5.4686 -3.2541 -7.9707>
        <9.124 -2.3525 -2.7576>
        <9.7447 -2.8673 -2.0393>
        <11.0334 -4.5946 -2.5482>
        <10.4638 -3.6369 -2.0941>
        <9.4803 -2.7394 -0.5615>
        <9.7314 -2.9219 -0.1679>
        <10.2895 -5.0307 1.4756>
        <10.4081 -5.868 1.759>
        <11.1205 -5.1768 -0.3413>
        <11.4399 -6.0936 -1.7183>
        <11.5924 -5.8214 -3.9115>
        <11.873 -6.7762 -3.1043>
        <11.8975 -6.9736 -2.059>
        <11.2379 -9.315 0.9643>
        <11.5589 -9.4529 -0.0004>
        <11.5746 -10.1239 -0.1627>
        <12.0883 -8.9035 -1.7677>
        <11.64 -11.5188 -0.7334>
        <10.9527 -14.0329 -1.6177>
        <8.465 -14.3041 -2.5794>
        <9.426 -14.6613 -2.9991>
        <10.1172 -14.8152 -3.2213>
        <10.8513 -14.4588 -3.4228>
        <11.1158 -13.9267 -1.6398>
        <12.0812 -11.6213 -3.5899>
        <12.039 -8.9807 -2.9138>
        <12.0542 -9.3141 -3.3373>
        <12.1631 -9.3239 -3.894>
        <11.5173 -11.489 -6.3305>
        <11.2728 -6.8754 -5.8681>
        <10.4383 -6.4982 -7.5813>
        <10.8015 -7.0742 -7.2848>
        <11.3066 -7.7835 -6.829>
        <11.2819 -8.1175 -6.8328>
        <9.7055 -7.1585 -8.7524>
        <10.3832 -7.796 -8.446>
        <7.204 -4.5191 -8.7661>
        <5.9048 -4.4661 -9.0342>
        <6.7521 -5.2665 -9.174>
        <9.175 -6.989 -9.1245>
        <8.787 -8.0223 -10.0029>
        <8.2775 -8.0567 -10.2017>
        <9.8983 -8.4261 -9.0853>
        <9.5609 -9.6136 -9.7234>
        <10.8857 -8.6929 -7.6597>
        <11.3206 -9.9343 -7.1412>
        <10.756 -10.9678 -8.5076>
        <10.4235 -11.0485 -9.0342>
        <10.405 -13.8887 -7.666>
        <10.563 -12.2617 -8.1593>
        <10.1839 -13.5238 -8.4774>
        <10.4568 -12.6583 -8.746>
        <9.2744 -9.5157 -10.12>
        <6.7583 -5.5624 -9.425>
        <5.5575 -5.1017 -9.576>
        <0.0112 -2.5226 -8.1847>
        <1.9244 -3.7782 -9.4198>
        <2.436 -4.2689 -9.7927>
        <0.7782 -3.2449 -9.2118>
        <0.776 -4.0473 -9.7997>
        <-3.9062 -2.244 -7.3802>
        <-3.2275 -3.3957 -8.7089>
        <-3.383 -3.6033 -8.8522>
        <-5.5793 -3.4449 -8.1107>
        <-6.8237 -4.9943 -8.9542>
        <-7.7263 -5.8134 -9.2911>
        <-6.316 -5.8505 -9.5898>
        <-8.7493 -6.3825 -9.1791>
        <-10.7418 -5.9864 -6.6088>
        <-10.6919 -6.1796 -6.9556>
        <-9.8274 -6.443 -8.0947>
        <-10.9692 -6.892 -6.9274>
        <-9.2553 -8.3409 -9.6592>
        <-5.9258 -9.2444 -11.7032>
        <-6.0116 -8.838 -11.4085>
        <-5.8881 -6.3448 -10.6053>
        <-4.4388 -5.2276 -10.1957>
        <-0.5467 -4.3509 -10.1971>
        <0.0399 -5.0484 -10.6561>
        <0.5641 -5.6957 -11.0021>
        <2.2233 -4.8176 -10.3128>
        <2.3847 -5.3534 -10.7593>
        <2.7514 -4.9737 -10.3723>
        <2.213 -5.5623 -10.8551>
        <3.4116 -5.6068 -10.6399>
        <5.8681 -6.3159 -10.5372>
        <3.3474 -6.7234 -11.2706>
        <5.0582 -9.7789 -11.5856>
        <0.2944 -6.6879 -11.4962>
        <-3.0698 -6.9238 -11.4166>
        <-7.2365 -12.3417 -11.2006>
        <-9.0314 -10.4739 -10.5676>
        <-8.9125 -11.3103 -10.9227>
        <-9.7323 -12.8767 -9.7859>
        <-8.4439 -12.6319 -10.8559>
        <-7.631 -12.7353 -11.0745>
        <-7.7519 -13.5501 -10.1945>
        <-8.3558 -13.9372 -9.9827>
        <-8.8918 -13.5649 -10.0784>
        <-8.8853 -13.8594 -9.7513>
        <-8.5881 -15.0561 -6.7376>
        <-8.6382 -15.0381 -6.0072>
        <-9.8055 -14.9072 -5.1185>
        <-10.4472 -14.8785 -5.2202>
        <-10.1949 -14.6943 -4.218>
        <-9.9557 -14.4715 -2.3369>
        <-10.4066 -14.5891 -2.6803>
        <-11.5071 -13.1791 -2.354>
        <-11.5296 -12.9902 -2.0961>
        <-11.9377 -12.5168 -3.5449>
        <-11.832 -13.3433 -4.1048>
        <-11.2026 -14.0965 -4.7608>
        <-11.1877 -13.8064 -5.6667>
        <-11.8658 -11.1912 -5.5235>
        <-10.6251 -12.362 -7.9496>
        <-10.8537 -13.4981 -7.4>
        <-10.4411 -14.3496 -6.8901>
        <-9.3995 -14.779 -7.2281>
        <-9.5437 -13.7175 -8.9455>
        <-9.8962 -12.6843 -9.48>
        <-10.3227 -12.0882 -8.8218>
        <-10.1302 -10.4419 -9.2529>
        <-10.7213 -7.4506 -7.5734>
        <-11.1484 -8.805 -7.3305>
        <-11.2575 -9.0526 -7.2537>
        <-11.8355 -10.3052 -5.8396>
        <-12.0822 -10.1671 -4.5465>
        <-11.5121 -8.4851 -6.1943>
        <-12.1535 -11.56 -4.061>
        <-12.0585 -10.566 -3.1025>
        <-10.6911 -12.6475 0.2581>
        <-11.3001 -12.0219 -0.1479>
        <-11.1751 -12.0464 0.3465>
        <-11.203 -10.8582 0.5984>
        <-11.0742 -10.5994 1.1627>
        <-10.8957 -9.6576 1.4901>
        <-11.8709 -11.0397 -1.9771>
        <-12.0549 -8.8434 -2.7951>
        <-11.8863 -8.5755 -5.1008>
        <-11.7679 -7.6992 -5.5137>
        <-10.8313 -4.926 -4.6198>
        <-11.9045 -7.3766 -3.6003>
        <-10.0718 -3.5855 -4.366>
        <-11.0402 -4.5843 -3.4653>
        <-11.8212 -6.6416 -2.6736>
        <-11.2985 -8.735 1.0892>
        <-11.4572 -8.2567 0.205>
        <8.5184 -14.7341 -5.4373>
        <9.745 -14.9533 -5.8595>
        <9.508 -15.0943 -6.2096>
        <7.4924 -13.9416 -9.2974>
        <7.1731 -10.6131 -11.5998>
        <6.7031 -10.7574 -11.7131>
        <8.2328 -13.2442 -10.5401>
        <-7.7429 -10.1832 5.0777>
        <-4.8669 -12.5959 5.8357>
        <-3.1972 -12.5394 6.5934>
        <-5.7194 -10.4427 6.1182>
        <-5.9077 -10.258 6.1609>
        <-5.7064 -10.2509 6.2259>
        <-1.5274 -13.228 6.6679>
        <-1.6226 -14.1215 6.3466>
        <-1.7196 -14.0513 6.3442>
        <3.8789 -12.8876 6.2352>
        <1.1668 -18.423 -10.2007>
        <1.7344 -17.9353 -10.2815>
        <0.2396 -18.7698 -9.9645>
        <-1.1965 -17.8405 -9.4385>
        <-5.7894 -15.1551 -10.222>
        <-6.0216 -14.1554 -10.6181>
        <-5.0865 -15.748 -10.3046>
        <-5.2488 -14.7562 -10.9884>
        <-7.039 -15.5582 -8.1556>
        <-5.818 -16.4844 -8.4481>
        <-2.6264 -17.8806 -7.2345>
        <-3.4635 -17.9878 -7.3642>
        <-2.3202 -18.1406 -4.9637>
        <-1.322 -18.075 -6.5114>
        <-4.8312 -17.5309 -6.5584>
        <-4.871 -17.5337 -5.9865>
        <-7.7382 -15.1539 -6.9992>
        <-7.7323 -15.3824 -6.2399>
        <-5.075 -17.5392 -5.9781>
        <-5.2289 -17.5394 -5.7594>
        <-5.329 -17.1495 -4.1277>
        <-4.5313 -17.6114 -4.0315>
        <-4.6439 -17.3995 -2.9162>
        <-3.1555 -17.7972 -2.9908>
        <-2.3083 -17.46 -2.1938>
        <-0.0457 -18.1533 -5.5753>
        <0.1429 -18.0912 -4.6295>
        <0.5905 -18.1471 -5.8872>
        <0.0929 -18.0052 -3.9491>
        <0.6359 -17.9354 -3.6498>
        <1.0891 -18.0845 -4.9401>
        <1.8179 -17.9873 -3.3935>
        <1.9835 -17.9569 -6.9991>
        <1.4426 -18.0016 -7.3253>
        <0.716 -18.0153 -7.7592>
        <-1.4925 -17.9561 -7.6577>
        <-2.5652 -17.724 -8.0542>
        <1.6186 -17.8784 -8.1578>
        <1.6162 -17.728 -9.1249>
        <-0.5443 -17.8692 -9.0277>
        <-0.9418 -16.9828 -11.5634>
        <1.771 -17.1305 -10.9612>
        <3.1037 -17.2776 -9.4368>
        <3.5412 -17.3177 -8.7711>
        <0.976 -16.8312 -11.7542>
        <0.7344 -16.7492 -11.9084>
        <-3.459 -16.861 -10.1589>
        <-3.7087 -16.6536 -10.3368>
        <0.6839 -15.3781 -12.7808>
        <2.2675 -16.6806 -11.3798>
        <2.0412 -14.8641 -12.6931>
        <2.6422 -15.9589 -11.8905>
        <3.4757 -16.4689 -10.8042>
        <4.5028 -15.5945 -10.9556>
        <4.182 -16.9372 -8.9246>
        <3.3132 -17.8658 -7.3217>
        <3.715 -17.6785 -7.7518>
        <5.1652 -17.3176 -6.9458>
        <5.1288 -16.9299 -8.422>
        <5.576 -15.5951 -9.9571>
        <4.1906 -14.7259 -11.7333>
        <5.1602 -14.3065 -11.2777>
        <6.4194 -13.1427 -10.6055>
        <7.3187 -14.999 -7.9022>
        <6.9654 -16.1625 -5.1478>
        <3.0283 -18.1387 -4.6413>
        <3.9931 -17.7462 -4.0863>
        <5.5859 -17.0913 -3.839>
        <3.8344 -17.8647 -4.6548>
        <6.4353 -16.535 -4.6966>
        <6.9369 -15.9143 -3.4875>
        <7.7269 -14.6182 -2.559>
        <7.0426 -15.5598 -2.695>
        <6.2397 -15.2946 -0.9763>
        <4.6945 -17.13 -2.3263>
        <5.0125 -16.7708 -2.0672>
        <5.4651 -16.1443 -1.3045>
        <4.5337 -15.8956 -0.368>
        <2.0375 -16.2932 0.176>
        <-7.7068 -14.815 -2.8046>
        <-6.7539 -15.9344 -2.4531>
        <-6.306 -16.1437 -2.564>
        <-6.1057 -16.2609 -2.485>
        <-6.6281 -14.6691 -0.9164>
        <-5.2577 -15.1938 -0.2142>
        <-5.3301 -16.53 -1.9067>
        <-5.1673 -16.7157 -2.1386>
        <-5.7294 -16.9415 -4.2144>
        <-5.9404 -16.689 -3.5733>
        <-7.2506 -15.9565 -3.9235>
        <-7.8172 -15.3101 -5.6854>
        <-8.2188 -14.687 -5.2045>
        <5.2138 -12.6541 -11.6372>
        <4.0201 -13.5926 -12.1688>
        <4.5036 -14.1781 -11.7796>
        <4.7486 -11.6487 -11.9191>
        <4.0797 -13.0544 -12.2399>
        <4.007 -9.6567 -11.9292>
        <3.19 -11.7272 -12.5404>
        <3.5505 -12.0289 -12.4672>
        <1.6833 -9.8318 -12.49>
        <1.2077 -8.7697 -12.1147>
        <0.3587 -9.6895 -12.5216>
        <-0.4093 -7.6583 -11.6579>
        <-2.457 -8.8038 -11.9861>
        <-4.7093 -10.2906 -11.7645>
        <-3.9268 -12.4665 -12.3386>
        <-3.9795 -14.0669 -12.0761>
        <-3.3535 -12.5323 -12.5298>
        <-0.4408 -11.9668 -13.0219>
        <-1.2434 -14.3925 -12.9354>
        <0.0953 -12.9898 -13.0924>
        <2.9888 -13.2451 -12.6559>
        <2.7545 -14.671 -12.4748>
        <-2.134 -16.5348 -0.2428>
        <-3.6528 -17.1975 -1.8115>
        <-4.2425 -16.6704 -1.1576>
        <-3.6398 -16.5573 -0.7706>
        <1.5901 -17.4156 -1.9803>
        <2.2077 -17.8369 -2.9085>
        <2.6201 -17.6755 -2.6321>
        <2.5653 -17.4433 -2.0907>
       ]}}
     Random Seed=0
    )
    Skeletal Mesh=$System.Skeletal Mesh
   ) {module=SkeletalMeshLocation002}
   &ScaleSpriteSize(Uniform Curve Scale=&FNTimeOfDayScalar(Night Scale=1.25, Morning Scale=1, Evening Scale=1)) {module=ScaleSpriteSize001}
   &ScaleColor(
    Scale Alpha=&Multiply_Float(
     A=&FloatFromCurve(
      FloatCurve=DataInterface() {propertyValues {Curve {keys [
          {interpMode=RCIM_Cubic, tangentMode=RCTM_User, time=0, value=0}
          {arriveTangent=-2.4058, interpMode=RCIM_Cubic, leaveTangent=-2.4058, tangentMode=RCTM_User, time=0.26, value=0.8421}
          {arriveTangent=-1.1775, interpMode=RCIM_Cubic, leaveTangent=-1.1775, time=0.4613, value=0.0645}
          {arriveTangent=-0.1197, interpMode=RCIM_Cubic, leaveTangent=-0.1197, time=0.5898, value=0.4537}
          <at 1 0 RCIM_Cubic>
         ]}}}
      Scale Curve=&RandomRangeFloat(Minimum=0.1, Maximum=0.5)
     )
     B=&FloatFromCurve(
      FloatCurve=DataInterface() {propertyValues {Curve {keys [{tangentMode=RCTM_User, time=0, value=0}, {tangentMode=RCTM_User, time=1, value=1}]}}}
      Scale Curve=&FloatFromCurve(
       FloatCurve=DataInterface() {propertyValues {Curve {keys [
           {arriveTangentWeight=0.1667, interpMode=RCIM_Cubic, leaveTangentWeight=0.1667, tangentMode=RCTM_User, time=1.4, value=0}
           {tangentMode=RCTM_User, time=2, value=1}
          ]}}}
       CurveIndex=$Emitter.Age
      )
     )
    )
    Scale RGB=&VectorFromFloat(Value=&RandomRangeFloat(Minimum=0.7, Maximum=0.9, Evaluation Type=::NewEnumerator1))
    ScaleA=true
    ScaleRGB=true
   ) {module=ScaleColor001}
   &ScaleColor(
    Scale Alpha=$System.TODAlpha
    Scale RGB=&MakeVectorFromLinearColorRGB(Color=$System.TODColor)
    ScaleA=true
    ScaleRGB=true
   )
   /SetVariables(
    Particles.CenterFlareColor:LinearColor=&MultiplyLinearColor(
     Linear Color=$Particles.Color
     Scale Factor=&MakeVector4_FromLinearColor(LinearColor=<rgba 8.0099 9.8007 10 1>)
    )
    Particles.CenterFlareSize:Vector2f=&Multiply_Vector2(A=$Particles.SpriteSize, B=&Vector2DFromFloat(Value=0.5))
    Particles.CoronaFlareColor:LinearColor=&MultiplyLinearColor(
     Linear Color=$Particles.Color
     Scale Factor=&MakeVector4_FromLinearColor(LinearColor=<rgba 0.15 0.15 0.15 1>)
    )
    Particles.CoronaFlareSize:Vector2f=&Multiply_Vector2(A=$Particles.Initial.SpriteSize, B=&Vector2DFromFloat(Value=0.75))
    Particles.FlickerFlareColor:LinearColor=&MultiplyLinearColor(
     Linear Color=$Particles.Color
     Scale Factor=&MakeVector4_FromLinearColor(LinearColor=<rgba 0.9441 0.8079 1 1>)
    )
   ) {isSetParameters=true}
  ]
  rendererClasses [&NiagaraSpriteRendererProperties]
  renderers [
   {rendererClass=&NiagaraSpriteRendererProperties}
   {rendererClass=&NiagaraSpriteRendererProperties, rendererIndex=1}
   {rendererClass=&NiagaraSpriteRendererProperties, rendererIndex=2}
  ]
 }}
summary {
 systemName=NS_Backpack_F_MED_Pipershelf_Sparkles
 emitters [{emitterName=Sparkles, rendererClasses [&NiagaraSpriteRendererProperties]}]
 userVariables [
  User.bIsFrontEnd {type=NiagaraBool}
  User.bIsFrontEndPreview {type=NiagaraBool}
  User.Lifetime_Scale {type=NiagaraFloat, defaultValue=0, description=""}
 ]
}
